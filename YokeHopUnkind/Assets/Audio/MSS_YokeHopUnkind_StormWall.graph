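# Looping quad MetaSound for a storm wall, mixing close and distant stereo loops from left/right sound-spline distances and angles.
# Smoothed power-curve attenuation sets each side's level, with exponents skewed by the left/right distance difference.
# The distant loop starts on play; an ADSR gates the close loop by audibility and bIsMoving, also bending its pitch.
# Angles crossfade each mixed side between front and surround channels.
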MetaSoundSource /YokeHopUnkind/Assets/Audio/MSS_YokeHopUnkind_StormWall
inputs [
 Attenuation.Close.FalloffDistance {type=Float}
 Attenuation.Close.InnerRadius {type=Float}
 Attenuation.Close.Intensity {type=Float}
 Attenuation.Distant.FalloffDistance {type=Float}
 Attenuation.Distant.InnerRadius {type=Float}
 Attenuation.Distant.Intensity {type=Float}
 bIsMoving {type=Bool}
 MasterVolume {type=Float}
 Mix.Close {type=Float}
 Mix.Distant {type=Float}
 SoundSpline.Angle.Left {type=Float}
 SoundSpline.Angle.Right {type=Float}
 SoundSpline.Distance.Left {type=Float}
 SoundSpline.Distance.Right {type=Float}
 SoundSpline.SampleWidth {type=Float}
 UE.Listener.Orientation.Azimuth {type=Float}
 UE.Listener.Orientation.Elevation {type=Float}
 UE.Source.OnPlay {type=Trigger}
 Wave.Close {type=WaveAsset}
 Wave.Distant {type=WaveAsset}
]
interfaces [UE.Listener.Orientation:1.0, UE.OutputFormat.Quad:1.0, UE.Source:1.1]
outputs [
 "UE.OutputFormat.Quad.Audio:0" {type=Audio}
 "UE.OutputFormat.Quad.Audio:1" {type=Audio}
 "UE.OutputFormat.Quad.Audio:4" {type=Audio}
 "UE.OutputFormat.Quad.Audio:5" {type=Audio}
]
pages [Default {
  input_defaults {
   Attenuation.Close.FalloffDistance=6000
   Attenuation.Close.InnerRadius=500
   Attenuation.Close.Intensity=5
   Attenuation.Distant.FalloffDistance=6000
   Attenuation.Distant.InnerRadius=500
   Attenuation.Distant.Intensity=5
   MasterVolume=0.752
   Mix.Close=1.2503
   Mix.Distant=1
   SoundSpline.Angle.Left=0
   SoundSpline.Angle.Right=0
   SoundSpline.Distance.Left=0
   SoundSpline.Distance.Right=0
   SoundSpline.SampleWidth=200
   UE.Listener.Orientation.Azimuth=0
   UE.Listener.Orientation.Elevation=0
   UE.Source.OnPlay=null
   Wave.Close=/Game/Athena/Sounds/Shield/Shield_PowerLoop_02
   Wave.Distant=/Game/Athena/Sounds/Shield/Shield_PowerLoop_03
   bIsMoving=null
  }
  nodes [
   InitVariable.Float()
   InitVariable.Float()
   n1 = UE.Multiply:Float(PrimaryOperand=^SoundSpline.SampleWidth) {also {defaults {AdditionalOperands=2}}}
   n2 = UE.Subtract:Float(PrimaryOperand=^SoundSpline.Distance.Left, AdditionalOperands=^SoundSpline.Distance.Right)
   n3 = VariableAccessor.Float(Variable=VariableMutator.Float(
     Variable=InitVariable.Float().Variable
     Value=MapRange.MapRange:Float(In Range B=@n1.Out, In=@n2.Out) {also {defaults {Out Range A=0.75, Out Range B=1.25}}}.Out Value
    ).Variable)
   n4 = UE.Power:Float(
    PrimaryOperand=MapRange.MapRange:Float(
     In Range A=^Attenuation.Close.InnerRadius
     In=^SoundSpline.Distance.Left
     In Range B=UE.Add:Float(AdditionalOperands=^Attenuation.Close.FalloffDistance, PrimaryOperand=^Attenuation.Close.InnerRadius).Out
    ) {also {defaults {Out Range A=1, Out Range B=0}}}.Out Value
    AdditionalOperands=UE.Multiply:Float(PrimaryOperand=^Attenuation.Close.Intensity, AdditionalOperands=@n3.Value).Out
   )
   n5 = VariableAccessor.Float(Variable=VariableMutator.Float(Variable=InitVariable.Float().Variable, Value=UE.InterpTo:Audio(Target=@n4.Out) {also {defaults {Interp Time=0.05}}}.Value).Variable)
   n6 = VariableAccessor.Float(Variable=@n5.Variable)
   n7 = VariableAccessor.Float(Variable=@n6.Variable)
   n8 = VariableAccessor.Float(Variable=@n7.Variable)
   n9 = UE.Multiply:Float(PrimaryOperand=^SoundSpline.SampleWidth) {also {defaults {AdditionalOperands=2}}}
   n10 = UE.Subtract:Float(AdditionalOperands=^SoundSpline.Distance.Left, PrimaryOperand=^SoundSpline.Distance.Right)
   n11 = VariableAccessor.Float(Variable=VariableMutator.Float(
     Variable=InitVariable.Float().Variable
     Value=MapRange.MapRange:Float(In Range B=@n9.Out, In=@n10.Out) {also {defaults {Out Range A=0.75, Out Range B=1.25}}}.Out Value
    ).Variable)
   n12 = UE.Power:Float(
    PrimaryOperand=MapRange.MapRange:Float(
     In Range A=^Attenuation.Close.InnerRadius
     In=^SoundSpline.Distance.Right
     In Range B=UE.Add:Float(AdditionalOperands=^Attenuation.Close.FalloffDistance, PrimaryOperand=^Attenuation.Close.InnerRadius).Out
    ) {also {defaults {Out Range A=1, Out Range B=0}}}.Out Value
    AdditionalOperands=UE.Multiply:Float(PrimaryOperand=^Attenuation.Close.Intensity, AdditionalOperands=VariableAccessor.Float(Variable=@n11.Variable).Value).Out
   )
   n13 = VariableAccessor.Float(Variable=VariableMutator.Float(Variable=InitVariable.Float().Variable, Value=UE.InterpTo:Audio(Target=@n12.Out) {also {defaults {Interp Time=0.05}}}.Value).Variable)
   n14 = VariableAccessor.Float(Variable=@n13.Variable)
   n15 = VariableAccessor.Float(Variable=@n14.Variable)
   n16 = VariableAccessor.Float(Variable=@n15.Variable)
   TriggerOnValueChange_0 = UE.Trigger On Value Change:Bool(Value=^bIsMoving)
   TriggerAnyTriggerAny2_1 = "TriggerAny.Trigger Any (2)"(In 0=^UE.Source.OnPlay, In 1=@TriggerOnValueChange_0.Trigger)
   n17 = TriggerCompare.Clamp:Bool(A=^bIsMoving, Compare=@TriggerAnyTriggerAny2_1.Out) {also {defaults {B="<MetaSoundBuilderResult.SUCCEEDED: 0>"}}}
   n18 = Convert.Float:Bool(Bool=^bIsMoving)
   TriggerControl_0 = UE.Trigger Control(
    Start Closed=Convert.Bool:Float(Float=UE.Subtract:Float(AdditionalOperands=@n18.Float) {also {defaults {PrimaryOperand=1}}}.Out).Bool
    Close=@n17.False
    Open=@n17.True
    Trigger In=UE.TriggerOnThreshold:Float(In=UE.Add:Float(PrimaryOperand=@n7.Value, AdditionalOperands=@n15.Value).Out) {also {defaults {Threshold=0.01}}}.Out
   )
   TriggerAnyTriggerAny2_0 = "TriggerAny.Trigger Any (2)"(
    In 0=TriggerCompare.Clamp:Float(A=UE.Add:Float(PrimaryOperand=@n8.Value, AdditionalOperands=@n16.Value).Out, Compare=@n17.True) {also {defaults {Type=3}}}.True
    In 1=@TriggerControl_0.Trigger Out
   )
   TriggerAnyTriggerAny2_2 = "TriggerAny.Trigger Any (2)"(
    In 1=UE.Reroute(Value=@n17.False).Value
    In 0=UE.TriggerOnThreshold:Float(In=UE.Add:Float(AdditionalOperands=@n14.Value, PrimaryOperand=@n6.Value).Out) {also {defaults {Threshold=0.01, Type=1}}}.Out
   )
   ADSREnvelopeADSREnvelope_0 = ADSR Envelope.ADSR Envelope:Float(Trigger Attack=@TriggerAnyTriggerAny2_0.Out, Trigger Release=@TriggerAnyTriggerAny2_2.Out) {also {defaults {Attack Curve=2, Attack Time=1.6, Decay Time=0, Release Curve=2, Release Time=2, Sustain Level=1}}}
   n19 = VariableAccessor.Float(Variable=VariableMutator.Float(Value=@ADSREnvelopeADSREnvelope_0.Out Envelope, Variable=InitVariable.Float().Variable).Variable)
   n20 = VariableAccessor.Float(Variable=@n19.Variable)
   n21 = InitVariable.Trigger()
   n22 = InitVariable.Trigger()
   WavePlayer_0 = UE.Wave Player:Stereo(
    Wave Asset=^Wave.Close
    Pitch Shift=UE.Multiply:Float(PrimaryOperand=UE.Subtract:Float(AdditionalOperands=@n20.Value) {also {defaults {PrimaryOperand=1}}}.Out) {also {defaults {AdditionalOperands=-2}}}.Out
    Stop=VariableAccessor.Trigger(Variable=VariableMutator.Trigger(Value=@ADSREnvelopeADSREnvelope_0.On Done, Variable=@n21.Variable).Variable).Value
    Play=VariableAccessor.Trigger(Variable=VariableMutator.Trigger(Value=@ADSREnvelopeADSREnvelope_0.On Attack Triggered, Variable=@n22.Variable).Variable).Value
   ) {also {defaults {Loop="<MetaSoundBuilderResult.SUCCEEDED: 0>"}, version [1,1]}}
   n23 = MapRange.MapRange:Float(
    In Range A=^Attenuation.Distant.InnerRadius
    In=^SoundSpline.Distance.Left
    In Range B=UE.Add:Float(AdditionalOperands=^Attenuation.Distant.FalloffDistance, PrimaryOperand=^Attenuation.Distant.InnerRadius).Out
   ) {also {defaults {Out Range A=1, Out Range B=0}}}
   n24 = UE.Multiply:Float(PrimaryOperand=^Attenuation.Distant.Intensity, AdditionalOperands=VariableAccessor.Float(Variable=@n3.Variable).Value)
   n25 = VariableMutator.Float(
    Variable=InitVariable.Float().Variable
    Value=UE.InterpTo:Audio(Target=UE.Power:Float(PrimaryOperand=@n23.Out Value, AdditionalOperands=@n24.Out).Out) {also {defaults {Interp Time=0.05}}}.Value
   )
   WavePlayer_1 = UE.Wave Player:Stereo(Play=^UE.Source.OnPlay, Wave Asset=^Wave.Distant) {also {defaults {Loop="<MetaSoundBuilderResult.SUCCEEDED: 0>"}, version [1,1]}}
   n26 = MapRange.MapRange:Float(
    In Range A=^Attenuation.Distant.InnerRadius
    In=^SoundSpline.Distance.Right
    In Range B=UE.Add:Float(AdditionalOperands=^Attenuation.Distant.FalloffDistance, PrimaryOperand=^Attenuation.Distant.InnerRadius).Out
   ) {also {defaults {Out Range A=1, Out Range B=0}}}
   n27 = UE.Multiply:Float(PrimaryOperand=^Attenuation.Distant.Intensity, AdditionalOperands=@n11.Value)
   n28 = VariableMutator.Float(
    Variable=InitVariable.Float().Variable
    Value=UE.InterpTo:Audio(Target=UE.Power:Float(PrimaryOperand=@n26.Out Value, AdditionalOperands=@n27.Out).Out) {also {defaults {Interp Time=0.05}}}.Value
   )
   AudioMixerAudioMixerStereo2_0 = "AudioMixer.Audio Mixer (Stereo, 2)"(
    In 0 R=UE.Multiply:Audio by Float(AdditionalOperands=UE.Multiply:Float(AdditionalOperands=@n19.Value, PrimaryOperand=@n13.Value).Out, PrimaryOperand=@WavePlayer_0.Out Right).Out
    Gain 1=UE.Multiply:Float(AdditionalOperands=^MasterVolume, PrimaryOperand=^Mix.Distant).Out
    Gain 0=UE.Multiply:Float(AdditionalOperands=^MasterVolume, PrimaryOperand=^Mix.Close).Out
    In 1 L=UE.Multiply:Audio by Float(AdditionalOperands=VariableAccessor.Float(Variable=@n25.Variable).Value, PrimaryOperand=@WavePlayer_1.Out Left).Out
    In 0 L=UE.Multiply:Audio by Float(AdditionalOperands=UE.Multiply:Float(AdditionalOperands=@n19.Value, PrimaryOperand=@n5.Value).Out, PrimaryOperand=@WavePlayer_0.Out Left).Out
    In 1 R=UE.Multiply:Audio by Float(AdditionalOperands=VariableAccessor.Float(Variable=@n28.Variable).Value, PrimaryOperand=@WavePlayer_1.Out Right).Out
   )
   n29 = MapRange.MapRange:Float(In=^SoundSpline.Angle.Left) {also {defaults {In Range B=180, Out Range A=1, Out Range B=0}}}
   n30 = VariableAccessor.Float(Variable=VariableMutator.Float(Variable=InitVariable.Float().Variable, Value=UE.InterpTo:Audio(Target=@n29.Out Value) {also {defaults {Interp Time=0.05}}}.Value).Variable)
   Multiply_1 = UE.Multiply:Audio by Float(PrimaryOperand=@AudioMixerAudioMixerStereo2_0.Out L, AdditionalOperands=@n30.Value)
   n31 = MapRange.MapRange:Float(In=^SoundSpline.Angle.Right) {also {defaults {In Range B=180, Out Range A=1, Out Range B=0}}}
   n32 = VariableAccessor.Float(Variable=VariableMutator.Float(Variable=InitVariable.Float().Variable, Value=UE.InterpTo:Audio(Target=@n31.Out Value) {also {defaults {Interp Time=0.05}}}.Value).Variable)
   Multiply_3 = UE.Multiply:Audio by Float(
    PrimaryOperand=@AudioMixerAudioMixerStereo2_0.Out R
    AdditionalOperands=UE.Subtract:Float(AdditionalOperands=@n32.Value) {also {defaults {PrimaryOperand=1}}}.Out
   )
   Multiply_5 = UE.Multiply:Audio by Float(PrimaryOperand=@AudioMixerAudioMixerStereo2_0.Out R, AdditionalOperands=@n32.Value)
   Multiply_7 = UE.Multiply:Audio by Float(
    PrimaryOperand=@AudioMixerAudioMixerStereo2_0.Out L
    AdditionalOperands=UE.Subtract:Float(AdditionalOperands=@n30.Value) {also {defaults {PrimaryOperand=1}}}.Out
   )
  ]
  links [
   {dst="@UE.OutputFormat.Quad.Audio:0", src=Multiply_1.Out}
   {dst="@UE.OutputFormat.Quad.Audio:5", src=Multiply_3.Out}
   {dst="@UE.OutputFormat.Quad.Audio:1", src=Multiply_5.Out}
   {dst="@UE.OutputFormat.Quad.Audio:4", src=Multiply_7.Out}
  ]
 }]
properties {
 assetImportData=&self:AssetImportData
 bLooping=true
 duration=10000
 numChannels=4
 outputFormat=Quad
 sampleRate=48000
 soundClassObject=/Game/Sounds/FortSoundClassesAndModes/Fort_Storms_PSM
 volume=1.2
}
