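# Material function that adjusts a Tiling value so a texture repeats a whole number of times across a run of segments, as its seamless-repeat name indicates.
# It multiplies Tiling, NumSegments and Segment UV Size into a total repeat count, floors it, and scales Tiling by the floored-to-raw ratio.
# The result rounds tiling down, reaching zero for positive totals below one, with no divide-by-zero guard.
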MaterialFunction /YokeHopUnkind/Assets/MF_SegmentCoord_SeamlessRepeat
graph {nodes [
  n1 = FunctionInput() {inputName=Tiling, inputType=FunctionInput_Scalar}
  n2 = FunctionInput() {bUsePreviewValueAsDefault=true, inputName=NumSegments, inputType=FunctionInput_Scalar, sortPriority=1, previewValue=<150 0 0 1>}
  n3 = Multiply(
   A=Multiply(A=@n1, B=@n2)
   B=FunctionInput() {bUsePreviewValueAsDefault=true, inputName=Segment UV Size, inputType=FunctionInput_Scalar, sortPriority=2}
  )
  n4 = Floor(None=@n3)
  FunctionOutput(None=Multiply(A=@n1, B=Divide(A=@n4, B=@n3)))
 ]}
