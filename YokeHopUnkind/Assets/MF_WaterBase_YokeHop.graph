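# MF_WaterBase_YokeHop is a water material function that sets emissive color and opacity on material attributes, previewed as Unlit.
# Panning turbulence, wave-noise and normal textures supply the surface motion, passed between stages through named reroutes.
# Emissive blends three base colors and a depth-based intersection color, then adds wave color scaled by a BaseNoise input.
# Opacity follows depth behind the translucency, with Normal, HeightMask and turbulence as extra outputs.
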MaterialFunction /YokeHopUnkind/Assets/MF_WaterBase_YokeHop
graph {nodes [
  n1 = ShadingModel() {shadingModel=MSM_Unlit}
  n2 = GetMaterialAttributes(None=FunctionInput(Preview=MakeMaterialAttributes!(
     BaseColor
     Metallic
     Specular
     Roughness
     Anisotropy
     EmissiveColor
     Opacity
     OpacityMask
     Normal
     Tangent
     WorldPositionOffset
     SubsurfaceColor
     ClearCoat
     ClearCoatRoughness
     AmbientOcclusion
     Refraction
     CustomizedUVs_0
     CustomizedUVs_1
     CustomizedUVs_2
     CustomizedUVs_3
     CustomizedUVs_4
     CustomizedUVs_5
     CustomizedUVs_6
     CustomizedUVs_7
     PixelDepthOffset
     ShadingModel=@n1
     Displacement
    )) {bUsePreviewValueAsDefault=true, inputType=FunctionInput_MaterialAttributes}) {attributeGetTypes [#1, #2]}
  FunctionOutput(None=SetMaterialAttributes!(MaterialAttributes=@n2.MaterialAttributes, EmissiveColor=NamedRerouteUsage(), Opacity=NamedRerouteUsage()) {attributeSetTypes [#1, #2]})
  FunctionOutput(None=NamedRerouteUsage()) {outputName=Normal}
  n3 = WorldPosition()
  n4 = ScalarParameter() {defaultValue=15000, parameterName=HeightGradientScale}
  n5 = Saturate(None=ComponentMask(None=Divide(A=@n3.XYZ, B=@n4)) {b=true})
  FunctionOutput(None=NamedRerouteDeclaration(None=OneMinus(None=@n5)) {name=HeightMask, nodeColor=<rgba 0.875 1 0 1>}) {outputName=HeightMask}
  n6 = VectorParameter() {group=Water, parameterName=TurbulentNoiseTiling, defaultValue=<rgba 4 8 4 1>}
  n7 = Panner(
   Coordinate=Multiply(A=TextureCoordinate(), B=ComponentMask(None=@n6.RGB) {g=true, r=true})
   Time=Time()
   Speed=ComponentMask(None=VectorParameter() {group=Water, parameterName=TurbulentNoiseSpeed, defaultValue=<rgba -0.05 0.2 0 1>}.RGB) {g=true, r=true}
  )
  FunctionOutput(None=NamedRerouteDeclaration(None=TextureSample(UVs=@n7) {samplerSource=SSM_Wrap_WorldGroupSettings, texture=/Game/Effects/Fort_Effects/Textures/Tile/T_GaussianNoise_01_M}.R) {name=TurbulentNoise, nodeColor=<rgba 0.2891 0 1 1>}) {outputName=Water Turbulence}
  Multiply()
  n8 = VectorParameter() {group=Water, parameterName=WaveNoiseTiling, defaultValue=<rgba 15 35 10 1>}
  n9 = Add(A=Multiply(A=TextureCoordinate(), B=ComponentMask(None=@n8.RGB) {g=true, r=true}), B=Multiply(A=NamedRerouteUsage()) {constB=0.1})
  n10 = ComponentMask(None=VectorParameter() {group=Water, parameterName=WaveNoiseSpeed, defaultValue=<rgba 0.05 -0.2 0 1>}.RGB) {g=true, r=true}
  NamedRerouteDeclaration(None=TextureSample(UVs=Panner(Coordinate=@n9, Speed=@n10)) {samplerType=SAMPLERTYPE_Normal, texture=/FortWater/Textures/T_Water_N}.RGB) {name=WaveNormal, nodeColor=<rgba 1 0.4922 0 1>}
  n11 = Time()
  NamedRerouteDeclaration(None=TextureSample(UVs=Panner(Coordinate=@n9, Time=@n11, Speed=@n10)) {samplerSource=SSM_Wrap_WorldGroupSettings, texture=/Game/Effects/Fort_Effects/Textures/Tile/T_Tile_WaveFlame}.R) {name=WaveNoise, nodeColor=<rgba 0.0312 1 0 1>}
  NamedRerouteDeclaration(None=@n2.EmissiveColor) {name=IN_Emissive, nodeColor=<rgba 1 0 0.2344 1>}
  NamedRerouteDeclaration(None=@n2.Opacity) {name=IN_Opacity, nodeColor=<rgba 0 0.5547 1 1>}
  n12 = VectorParameter() {group=Water, parameterName=BaseColorB, defaultValue=<rgba 0.241 1 0.5742 1>}
  n13 = VectorParameter() {group=Water, parameterName=BaseColorA, defaultValue=<rgba 0.6927 0.9433 1 1>}
  n14 = NamedRerouteUsage()
  n15 = LinearInterpolate(
   A=LinearInterpolate(
    A=LinearInterpolate(A=@n12.RGB, B=@n13.RGB, Alpha=@n14)
    B=VectorParameter() {group=Water, parameterName=BaseColorC, defaultValue=<rgba 0 0.1969 0.4 1>}.RGB
    Alpha=NamedRerouteUsage()
   )
   B=VectorParameter() {parameterName=IntersectionColor, defaultValue=<rgba 0.2936 3.4212 8 1>}.RGB
   Alpha=NamedRerouteUsage()
  )
  n16 = FunctionInput() {inputName=BaseNoise, inputType=FunctionInput_Scalar}
  n17 = NamedRerouteUsage()
  n18 = ScalarParameter() {defaultValue=0.212, group=Water, parameterName=WaveIntensity}
  n19 = Multiply(A=@n18) {constB=0.5}
  n20 = NamedRerouteUsage()
  n21 = Multiply(
   A=Multiply(A=Multiply(A=@n16, B=@n17), B=LinearInterpolate(A=@n19, B=@n18, Alpha=@n20)) {constB=0.1}
   B=VectorParameter() {group=Water, parameterName=WaveColor, defaultValue=<rgba 0.1301 2 3 1>}.RGB
  )
  NamedRerouteDeclaration(None=Add(A=Add(A=@n15, B=@n21), B=NamedRerouteUsage())) {name=OUT_Emissive, nodeColor=<rgba 0.8047 0 1 1>}
  n22 = ScalarParameter() {defaultValue=0.2427, group=Water, parameterName=OpacityMin}
  n23 = ScalarParameter() {defaultValue=0.786, group=Water, parameterName=OpacityMax}
  n24 = NamedRerouteDeclaration(None=MaterialFunctionCall!(
    Distance Bias=Constant()
    Distance Divisor=ScalarParameter() {defaultValue=12655.3203, group=Water, parameterName=DepthDistance}
    Scene Depth
    Pixel Depth
   ) {materialFunction=/Game/Effects/Fort_Effects/Materials/_Functions/MF_DistanceBehindTranslucency_01}.Normalized) {name=DepthDistance, nodeColor=<rgba 0 1 0.0859 1>}
  NamedRerouteDeclaration(None=Min(A=LinearInterpolate(A=@n22, B=@n23, Alpha=@n24), B=NamedRerouteUsage())) {name=OUT_Opacity, nodeColor=<rgba 0 1 0.4844 1>}
  n25 = Constant()
  n26 = Constant() {r=250}
  NamedRerouteDeclaration(None=Power(
    Base=MaterialFunctionCall(Distance Bias=@n25, Distance Divisor=@n26) {materialFunction=&MF_DistanceBehindTranslucency_01}.Normalized Inverted
    Exp=Constant() {r=2.7367}
   )) {name=IntersectionMask, nodeColor=<rgba 0 0.25 1 1>}
  TextureCoordinate() {uTiling=24, vTiling=4}
 ]}
