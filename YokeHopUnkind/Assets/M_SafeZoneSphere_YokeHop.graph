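# Translucent, two-sided unlit material for a safe-zone boundary sphere, drawn as a glowing blue energy wall.
# Emissive layers scrolling scanlines, animated world-aligned caustics, a rippling grid and a depth-based intersection glow over a base color.
# Opacity combines these masks with height, near and far fades, while an If pushes distant points far down, apparently to hide them.
# Quality and feature-level switches simplify color and opacity on low-end targets.
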Material /YokeHopUnkind/Assets/M_SafeZoneSphere_YokeHop
graph {
 properties {
  bEnableResponsiveAA=true
  bUsedWithNanite=true
  bUsedWithSplineMeshes=true
  blendMode=BLEND_Translucent
  pixelDepthOffsetMode=PDOM_Legacy
  shadingModel=MSM_Unlit
  twoSided=true
  usedShadingModels=Unlit
 }
 nodes [
  Constant() {r=0.261}
  Constant() {r=1.25}
  FeatureLevelSwitch()
  QualitySwitch()
  n1 = MaterialFunctionCall(Vector 2=ComponentMask(None=WorldPosition() {worldPositionShaderOffset=WPT_CameraRelative}.XYZ) {g=true, r=true}) {materialFunction=/Engine/Functions/Engine_MaterialFunctions02/Utility/VectorLength}
  n2 = ScalarParameter() {defaultValue=44900, parameterName=FarFade}
  n3 = ScalarParameter() {defaultValue=1.1, parameterName=Buffer}
  n4 = Constant()
  Reroute(None=If(
    A=@n1.V2 Length
    B=Multiply(A=@n2, B=@n3) {constB=1.36, desc=add some buffer}
    "A > B"=Constant3Vector() {constant=<rgba 0 0 -999999 0>}
    "A == B"=@n4
    "A < B"=@n4
   ))
  n5 = Add(A=WorldPosition().XYZ, B=Multiply(A=Time() {bOverride_Period=true, period=1}, B=Constant()))
  n6 = Constant() {r=0.05}
  n7 = TextureSample(UVs=ComponentMask(None=Multiply(A=@n5, B=@n6)) {b=true}) {samplerType=SAMPLERTYPE_LinearGrayscale, texture=/Game/Athena/SafeZone/ThinLines}
  n8 = Add(A=PixelDepth().R, B=Constant() {r=-500})
  n9 = Constant() {r=3500}
  n10 = OneMinus(None=Saturate(None=Divide(A=@n8, B=@n9)))
  n11 = Multiply(A=@n7.R, B=@n10) {desc=lines * distance fade}
  n12 = Add(
   A=Multiply(A=@n11, B=VectorParameter() {parameterName=scanlines Color, defaultValue=<rgba 0.0928 0.2583 0.8 1>}.RGB)
   B=VectorParameter() {parameterName=Day Color Far, defaultValue=<rgba 0.025 0.4409 1 1>}.RGB
  )
  n13 = MaterialFunctionCall!(
   TextureObject=TextureObject() {samplerType=SAMPLERTYPE_Normal, texture=/Game/TheOutpost/Textures/TilingWeakPointNormal}
   TextureSize=VectorParameter() {parameterName=CausticWorldScale, defaultValue=<rgba 29400 36750 21000 0>}.RGB
   WorldPosition
   Export Float 4=StaticBool()
   In
   World Space Normal
   ProjectionTransitionContrast=Constant() {r=1}
  ) {materialFunction=/Game/Athena/SafeZone/WorldAlignedTexture_ScalePan}
  n14 = Multiply(A=Time() {desc=15.015014, period=1}, B=ScalarParameter() {defaultValue=0.08, desc=0.0666, parameterName=Sine Speed A_1})
  n15 = ComponentMask(None=Frac(None=Add(A=@n13.XY Texture, B=@n14))) {g=true}
  n16 = Constant() {r=0.5}
  n17 = OneMinus(None=Multiply(A=Distance(A=@n15, B=@n16)) {constB=2})
  n18 = Multiply(
   A=LinearInterpolate(A=@n17, B=Round(None=@n17), Alpha=ScalarParameter() {defaultValue=0.25, parameterName=Caustic Round LERP})
   B=VectorParameter() {parameterName=Caustics Color, defaultValue=<rgba 0.016 0.1184 0.4 1>}.RGB
  )
  n19 = Constant() {r=1}
  n20 = FeatureLevelSwitch(Default=@n17, ES2=@n19, ES3_1=@n19, SM4_REMOVED=@n17)
  n21 = ScalarParameter() {defaultValue=0.2, parameterName=Ripple to Grid Scale}
  n22 = MaterialFunctionCall(
   TextureObject=TextureObject() {samplerType=SAMPLERTYPE_LinearColor, texture=/Game/Gadgets/Assets/Teleporter/Textures/T_Teleporter_GridSquares_01}
   TextureSize=Constant() {r=512}
   Export Float 4=StaticBool()
   ProjectionTransitionContrast=Constant() {r=5}
  ) {materialFunction=&WorldAlignedTexture_ScalePan}
  n23 = ComponentMask(None=@n22.XY Texture) {r=true}
  n24 = Multiply(A=Time() {bOverride_Period=true, period=3.8314}, B=ScalarParameter() {defaultValue=0.261, parameterName=Sine Speed Grid})
  n25 = Sine(None=Add(A=Multiply(A=@n20, B=@n21), B=Add(A=@n23, B=@n24)))
  n26 = ComponentMask(None=@n22.XY Texture) {g=true}
  n27 = Multiply(A=@n10, B=Power(Base=Multiply(A=@n25, B=@n26), Exp=ScalarParameter() {defaultValue=5, parameterName=Grid Power}))
  n28 = Multiply(A=@n27, B=VectorParameter() {parameterName=Grid Color, defaultValue=<rgba 0.0143 0.063 0.15 0>}.RGB)
  n29 = Constant()
  n30 = Constant() {r=250}
  n31 = Multiply(
   A=Power(
    Base=MaterialFunctionCall!(Distance Bias=@n29, Distance Divisor=@n30, Scene Depth, Pixel Depth) {materialFunction=/Game/Effects/Fort_Effects/Materials/_Functions/MF_DistanceBehindTranslucency_01}.Normalized Inverted
    Exp=ScalarParameter() {defaultValue=6, parameterName=Intersection Edge Exponent}
   )
   B=VectorParameter() {parameterName=Intersection Edge Color, defaultValue=<rgba 0.5 3 40 1>}.RGB
  )
  n32 = Add(A=Add(A=@n12, B=Add(A=@n18, B=@n28) {desc=+ Grid Blocks}), B=@n31) {desc=+ intersection edge}
  n33 = FeatureLevelSwitch(Default=@n32, ES2=@n12, ES3_1=@n12, SM4_REMOVED=@n32)
  Reroute(None=QualitySwitch(Default=@n33, Low=@n12, High=@n33, Medium=@n33))
  Reroute(None=Constant3Vector() {constant=<rgba 1 0 0.734 1>})
  n34 = Add(A=PixelDepth().R, B=Constant() {r=-100})
  n35 = ScalarParameter() {defaultValue=300, parameterName=NearFade}
  n36 = ComponentMask(None=Divide(A=WorldPosition().XYZ, B=ScalarParameter() {defaultValue=23000, parameterName=HeightGradientScale})) {b=true}
  n37 = Add(A=Multiply(A=@n27, B=Constant() {r=0.4}), B=@n11)
  n38 = LinearInterpolate(
   A=Constant() {r=0.4}
   B=Constant() {r=0.7}
   Alpha=MaterialFunctionCall(Distance Bias=Constant(), Distance Divisor=Constant() {r=10000}) {materialFunction=&MF_DistanceBehindTranslucency_01}.Normalized
  )
  n39 = FeatureLevelSwitch(Default=@n38, ES3_1=Add(A=@n38, B=Constant() {r=-0.2}))
  n40 = Multiply(A=OneMinus(None=Saturate(None=@n36)), B=Add(A=@n31, B=Add(A=@n37, B=@n39)))
  n41 = Constant() {r=-100}
  n42 = Saturate(None=Divide(A=Add(A=@n1.V2 Length, B=@n41), B=@n2))
  n43 = Multiply(A=Saturate(None=Divide(A=@n34, B=@n35)), B=Multiply(A=QualitySwitch(Default=@n40, Low=@n39), B=OneMinus(None=@n42)))
  n44 = Add(A=@n43, B=@n7.R)
  n45 = FeatureLevelSwitch(Default=@n43, ES2=@n44, ES3_1=@n44, SM4_REMOVED=@n43, SM5=@n43, SM6=@n43)
  Reroute(None=Multiply(A=QualitySwitch(Default=@n45, Low=@n44, High=@n45, Medium=@n45), B=ScalarParameter() {defaultValue=0.45, parameterName=OpacityCoefficient}) {constB=0.3})
  ScalarParameter() {defaultValue=512, parameterName=GridWorldScale}
  ScalarParameter() {defaultValue=13, parameterName=GridProjContrast}
  ScalarParameter() {defaultValue=100000, parameterName=Saturation Divisor}
  ScalarParameter() {defaultValue=250, parameterName=Intersection Edge Divisor}
  ShadingPathSwitch()
 ]
}
