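# Translucent, unlit, two-sided material for a storm shield wall with animated caustics.
# Panning world-aligned textures, warped by noise, form caustics, sparks and distance-banded detail over a hue-shifted base with separate backside color.
# Depth-based edge glow and a wave infill driven by the collection's shrink-phase alpha add highlights, scaled by inverse eye adaptation and time of day.
# Opacity fades by height, distance and depth; quality and shading-path switches simplify low-end and mobile.
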Material /YokeHopUnkind/Assets/M_YokeHopKeeper_StormShield_Causticy
graph {
 properties {
  bCastRayTracedShadows=false
  bEnableResponsiveAA=true
  bUsedWithNanite=true
  bUsedWithSplineMeshes=true
  blendMode=BLEND_Translucent
  pixelDepthOffsetMode=PDOM_Legacy
  shadingModel=MSM_Unlit
  twoSided=true
  usedShadingModels=Unlit
 }
 nodes [
  n1 = NamedRerouteUsage()
  n2 = ScalarParameter() {defaultValue=1.2, parameterName=VolumeNoiseInfluence}
  n3 = MaterialFunctionCall!(
   TextureObject=TextureObjectParameter() {parameterName=UltraClose, texture=&/StormShield_CloseRange_T_CloseDistanceStorm_M}
   TextureSize=ScalarParameter() {defaultValue=1000, parameterName=CausticScaleH}
   WorldPosition
   Export Float 4=StaticBool()
   BumpOffsetHeight=ScalarParameter() {defaultValue=-0.7052, parameterName=UltraCloseLayerOffset}
   In=VectorParameter() {parameterName=PanningH, defaultValue=<rgba 0 0 0 1>}.RGB
   World Space Normal
   ProjectionTransitionContrast
  ) {materialFunction=/Game/Athena/SafeZone/WorldAlignedTexture_ScalePanBumpOffset}
  n4 = Add(
   A=Add(A=ComponentMask(None=@n3.XYZ Texture) {r=true}, B=Multiply(A=@n1, B=@n2) {constB=8})
   B=Multiply(A=Time(), B=ScalarParameter() {defaultValue=0.0161, parameterName=MovementSpeed})
  )
  Frac(None=@n4)
  n5 = Saturate(None=MaterialFunctionCall!(Sine Phase=@n4, Value 1, Value 2) {materialFunction=/Engine/Functions/Engine_MaterialFunctions02/Utility/Sine_Remapped}.Remapped Sine)
  MaterialFunctionCall!(In=@n5, Ramp Width) {materialFunction=/Engine/Functions/Engine_MaterialFunctions01/Math/SmoothCeil}
  n6 = Multiply(A=Time() {period=3.8314}, B=ScalarParameter() {defaultValue=0.157, parameterName=Sine Speed Grid})
  MaterialFunctionCall(Value=@n6) {materialFunction=&LinearSine}
  n7 = MaterialFunctionCall(
   TextureObject=TextureObjectParameter() {parameterName=caustictexture, samplerType=SAMPLERTYPE_LinearColor, texture=&/StormShield2_M}
   TextureSize=ScalarParameter() {defaultValue=4905.8989, parameterName=CausticScaleG}
   Export Float 4=StaticBool()
   BumpOffsetHeight=ScalarParameter() {defaultValue=-8.0327, parameterName=CloseLayerOffset}
   In=VectorParameter() {parameterName=Panning G, defaultValue=<rgba -0.01 0 0 1>}.RGB
  ) {materialFunction=&WorldAlignedTexture_ScalePanBumpOffset}
  n8 = WorldPosition()
  n9 = Multiply(A=ScalarParameter() {defaultValue=0.08, parameterName=CloseWarpIntensity}, B=NamedRerouteUsage())
  n10 = MaterialFunctionCall(
   TextureObject=TextureObjectParameter() {parameterName=StarTexture, samplerType=SAMPLERTYPE_LinearColor, texture=&StormShield2_M}
   TextureSize=ScalarParameter() {defaultValue=6313.064, parameterName=CausticScaleF}
   WorldPosition=Add(A=Add(A=@n8.XYZ, B=@n9), B=ScalarParameter() {defaultValue=423.5759, parameterName=OffsetA})
   Export Float 4=StaticBool()
   In=VectorParameter() {parameterName=Panning F, defaultValue=<rgba 0.0185 0.004 0 1>}.RGB
  ) {materialFunction=&WorldAlignedTexture_ScalePanBumpOffset}
  n11 = ComponentMask(None=@n10.XYZ Texture) {g=true}
  n12 = Multiply(A=Add(A=ComponentMask(None=@n7.XYZ Texture) {r=true}, B=ComponentMask(None=@n10.XYZ Texture) {r=true}), B=@n11)
  n13 = Multiply(A=ComponentMask(None=@n7.XYZ Texture) {g=true}, B=@n11)
  n14 = NamedRerouteUsage()
  n15 = Multiply(A=Multiply(A=Add(A=@n12, B=@n13), B=OneMinus(None=@n14)), B=ScalarParameter() {defaultValue=1.9093, parameterName=StarIntensity})
  n16 = Distance(A=WorldPosition() {worldPositionShaderOffset=WPT_CameraRelative}.XYZ, B=Constant())
  n17 = ScalarParameter() {defaultValue=5500, parameterName=Close Distance Radius}
  n18 = Saturate(None=OneMinus(None=Divide(A=@n16, B=@n17)))
  Multiply(A=@n15, B=@n18)
  Multiply()
  n19 = Divide(A=@n16, B=ScalarParameter() {defaultValue=1894.5015, parameterName=ultraclose Distance Radius})
  n20 = Multiply(A=@n19, B=@n19)
  n21 = Multiply(A=@n20, B=@n20)
  Multiply(A=@n21, B=@n21)
  n22 = TextureObjectParameter() {parameterName=caustictexture, samplerType=SAMPLERTYPE_LinearColor, texture=&StormShield2_M}
  n23 = Add(A=WorldPosition().XYZ, B=NamedRerouteUsage())
  n24 = StaticBoolParameter() {parameterName="Export Float 4 (SB)"}
  n25 = MaterialFunctionCall(
   TextureObject=@n22
   TextureSize=ScalarParameter() {defaultValue=22619.9902, parameterName=CausticScale}
   WorldPosition=Add(A=@n23, B=ScalarParameter() {defaultValue=423.5759, parameterName=OffsetA})
   Export Float 4=@n24
   In=VectorParameter() {parameterName=Panning A, defaultValue=<rgba -0.0065 0.0001 0.001 1>}.RGB
  ) {materialFunction=&WorldAlignedTexture_ScalePanBumpOffset}
  n26 = ComponentMask(None=@n25.XYZ Texture) {b=true}
  n27 = VectorParameter() {parameterName=Panning B, defaultValue=<rgba 0.004 -0.0012 0 1>}
  n28 = ScalarParameter() {defaultValue=3, parameterName=Backside Speed Multiplier}
  n29 = MaterialFunctionCall(
   TextureObject=@n22
   TextureSize=ScalarParameter() {defaultValue=29807.1113, parameterName=CausticScaleB}
   WorldPosition=@n23
   Export Float 4=@n24
   BumpOffsetHeight=ScalarParameter() {defaultValue=-3.5655, parameterName=Bump B}
   In=LinearInterpolate(A=Multiply(A=@n27.RGB, B=@n28), B=@n27.RGB, Alpha=TwoSidedSign())
  ) {materialFunction=&WorldAlignedTexture_ScalePanBumpOffset}
  n30 = ComponentMask(None=@n29.XYZ Texture) {g=true}
  NamedRerouteDeclaration(None=Divide(A=Add(A=@n26, B=@n30))) {name=MiddleMask, nodeColor=<rgba 0.6172 1 0 1>}
  n31 = Divide(A=WorldPosition().XYZ, B=ScalarParameter() {defaultValue=20584.8906, parameterName=CurlNoiseScale})
  n32 = Multiply(
   A=Multiply(A=Time() {period=10}, B=VectorParameter() {parameterName=CurlNoiseMotion, defaultValue=<rgba -0.02 -0.02 -0.02 1>}.RGB)
   B=ScalarParameter() {defaultValue=0.5057, parameterName=CurlNoiseSpeed}
  )
  NamedRerouteDeclaration(None=TextureSample(UVs=Add(A=@n31, B=@n32)) {samplerType=SAMPLERTYPE_LinearColor, texture=/Game/Effects/FogVolume/VT_Voronoi_LOW}.RGB) {name=volumeNoise, nodeColor=<rgba 0.7344 1 0 1>}
  n33 = Add(A=ComponentMask(None=@n25.XYZ Texture) {r=true}, B=ComponentMask(None=@n29.XYZ Texture) {r=true})
  NamedRerouteDeclaration(None=@n33) {name=Highlight, nodeColor=<rgba 1 0.7266 0 1>}
  n34 = OneMinus(None=NamedRerouteUsage())
  n35 = Multiply(A=NamedRerouteUsage(), B=ScalarParameter() {defaultValue=0.5824, parameterName=HighlightDamper})
  n36 = Power(Base=ComponentMask(None=NamedRerouteUsage()) {b=true}, Exp=ScalarParameter() {defaultValue=0.9818, parameterName=Spark Power 2})
  n37 = NamedRerouteUsage()
  n38 = QualitySwitch(Default=NamedRerouteUsage(), Low=@n37)
  n39 = Power(
   Base=Multiply(A=Multiply(A=@n35, B=@n36), B=ShadingPathSwitch(Default=@n38, Mobile=@n37))
   Exp=ScalarParameter() {defaultValue=2.2004, parameterName=Spark Power}
  )
  NamedRerouteDeclaration(None=QualitySwitch(Default=Multiply(A=@n34, B=@n39), Low=@n39)) {name=Caustic Spark, nodeColor=<rgba 0.4062 0 1 1>}
  NamedRerouteDeclaration(None=@n30) {name=staroffsetmask, nodeColor=<rgba 0 0.25 1 1>}
  n40 = Multiply(A=ComponentMask(None=@n25.XYZ Texture) {g=true}, B=@n30) {desc=old}
  n41 = Multiply(A=@n33, B=@n40)
  n42 = Power(
   Base=Multiply(A=Add(A=@n41, B=@n40) {desc=old}, B=ScalarParameter() {defaultValue=15.2684, parameterName=CausticMultiply})
   Exp=ScalarParameter() {defaultValue=0.5958, parameterName=CausticPower}
  ) {constExponent=1.1532}
  n43 = ScalarParameter() {defaultValue=40764.707, parameterName=Far Fade Dist}
  n44 = ScalarParameter() {defaultValue=10968.1035, parameterName=Distance Fade, sortPriority=8000000}
  n45 = LinearInterpolate(
   A=Multiply(A=@n42, B=ScalarParameter() {defaultValue=0.2312, group=Caustics, parameterName=Caustics Brightness Low})
   B=Multiply(A=@n42, B=ScalarParameter() {defaultValue=0.8717, group=Caustics, parameterName=Caustics Brightness}) {constB=2}
   Alpha=OneMinus(None=MaterialFunctionCall(Fade Length=@n43, Fade Offset=@n44) {materialFunction=&CameraDepthFade}.Result)
  )
  n46 = Multiply(A=@n45) {constB=0.3}
  n47 = MaterialFunctionCall(
   Fade Length=ScalarParameter() {defaultValue=6341.5488, parameterName=Near Fade Dist}
   Fade Offset=ScalarParameter() {defaultValue=8056.0239, parameterName=Distance Fade Near, sortPriority=8000000}
  ) {materialFunction=&CameraDepthFade}
  NamedRerouteDeclaration(None=QualitySwitch(Default=LinearInterpolate(A=@n46, B=@n45, Alpha=@n47.Result), Low=@n42)) {name=Caustic, nodeColor=<rgba 0.1016 1 0 1>}
  n48 = NamedRerouteUsage()
  n49 = ScalarParameter() {defaultValue=0.2, parameterName=Ripple to Grid Scale}
  n50 = MaterialFunctionCall!(
   TextureObject=TextureObjectParameter() {parameterName=CubeColors, texture=&/CloseRangeStormShield_M3}
   TextureSize=Constant() {r=512}
   WorldPosition
   Export Float 4=StaticBool()
   In
   World Space Normal
   ProjectionTransitionContrast=Constant() {r=5}
  ) {materialFunction=/Game/Athena/SafeZone/WorldAlignedTexture_ScalePan}
  n51 = ComponentMask(None=@n50.XY Texture) {r=true}
  n52 = Frac(None=@n6)
  n53 = Sine(None=Add(A=Multiply(A=@n48, B=@n49), B=Add(A=@n51, B=@n52)))
  n54 = ComponentMask(None=@n50.XY Texture) {g=true}
  NamedRerouteDeclaration(None=Power(Base=Multiply(A=@n53, B=@n54), Exp=ScalarParameter() {defaultValue=5, parameterName=Grid Power})) {name=Ultraclose, nodeColor=<rgba 0.8047 0 1 1>}
  NamedRerouteDeclaration(None=Saturate(None=@n21)) {name=UltraCloseRangeBlend, nodeColor=<rgba 0.7578 1 0 1>}
  n55 = VectorParameter() {parameterName=Distance Blend Color, defaultValue=<rgba 0.0061 0.3581 2 1>}
  n56 = VectorParameter() {parameterName=aaa, defaultValue=<rgba 0.8438 0.0352 0.3829 1>}
  n57 = NamedRerouteUsage()
  n58 = NamedRerouteUsage()
  n59 = Multiply(
   A=ShadingPathSwitch(Default=QualitySwitch(Default=@n57, Low=@n58), Mobile=@n58)
   B=ScalarParameter() {defaultValue=0.4173, parameterName=MiddleMaskBlend}
  )
  n60 = ScalarParameter() {defaultValue=22455.2344, parameterName=DistanceDivide}
  n61 = VectorParameter() {parameterName=Base Background Brightness, defaultValue=<rgba 0 0.2825 0.8113 0.5>}
  n62 = VectorParameter() {parameterName=Base Background Brightness B, defaultValue=<rgba 0.0494 0 1.2083 0.5>}
  n63 = Saturate(None=Power(Base=NamedRerouteUsage(), Exp=ScalarParameter() {defaultValue=1.3721, parameterName=asdav}))
  NamedRerouteDeclaration(None=LinearInterpolate(
    A=LinearInterpolate(A=@n61.RGB, B=@n62.RGB, Alpha=@n63)
    B=LinearInterpolate(A=@n55.RGB, B=@n56.RGB, Alpha=@n59)
    Alpha=MaterialFunctionCall!(Distance Bias, Distance Divisor=@n60, Scene Depth, Pixel Depth) {materialFunction=/Game/Effects/Fort_Effects/Materials/_Functions/MF_DistanceBehindTranslucency_01}.Clamped
   )) {name=Base Color Level, nodeColor=<rgba 1 0.4922 0 1>}
  n64 = Multiply(
   A=Constant4Vector() {constant=<rgba 4.1308 20 18.5846 1>}.RGB
   B=SphereMask(
    A=ScalarParameter() {defaultValue=6201.6055, parameterName=WaveTop}
    B=WorldPosition().Z
    Radius=ScalarParameter() {defaultValue=3432.531, parameterName=GlowRadius}
    Hardness=ScalarParameter() {defaultValue=-0.66, parameterName=Glowhardness}
   )
  )
  n65 = Multiply(A=Multiply(A=NamedRerouteUsage(), B=NamedRerouteUsage()) {constB=2}) {constB=0.66}
  NamedRerouteDeclaration(None=Multiply(A=Multiply(A=@n64, B=@n65), B=NamedRerouteUsage())) {name=Wave Infill Blend, nodeColor=<rgba 0.1719 0 1 1>}
  NamedRerouteDeclaration(None=@n26) {name=Noisymask01, nodeColor=<rgba 0 1 0.625 1>}
  n66 = ComponentMask(None=MaterialFunctionCall(
    TextureObject=TextureObjectParameter() {parameterName=NoiseTexture, samplerType=SAMPLERTYPE_Alpha, texture=&/StormShield2_Noise}
    TextureSize=ScalarParameter() {defaultValue=28368.6758, parameterName=CausticScaleC}
    Export Float 4=StaticBool() {value=true}
    In=VectorParameter() {parameterName=Panning C, defaultValue=<rgba 0.003 -0.005 -0.0044 1>}.RGB
   ) {materialFunction=&WorldAlignedTexture_ScalePanBumpOffset}.XYZ Texture) {a=true}
  NamedRerouteDeclaration(None=@n66) {name=PerlinA, nodeColor=<rgba 1 0.3047 0 1>}
  NamedRerouteDeclaration(None=SphereMask(A=CollectionParameter() {collection=&/MPC_YokeHopUnkind, parameterName=ShrinkPhaseCompletionAlpha}, B=Constant() {r=0.5}) {attenuationRadius=0.45, hardnessPercent=3}) {name=isWaveInfill, nodeColor=<rgba 1 0 0.0234 1>}
  NamedRerouteDeclaration(None=Divide(A=Add(A=@n26, B=@n30))) {name=mask3, nodeColor=<rgba 0.8984 1 0 1>}
  n67 = ScalarParameter() {parameterName=bias}
  n68 = ScalarParameter() {defaultValue=250, parameterName=Intersection Edge Divisor}
  NamedRerouteDeclaration(None=Power(
    Base=MaterialFunctionCall(Distance Bias=@n67, Distance Divisor=@n68) {materialFunction=&MF_DistanceBehindTranslucency_01}.Inverted
    Exp=ScalarParameter() {defaultValue=6, parameterName=Intersection Edge Exponent}
   )) {name=Edge Glow, nodeColor=<rgba 0.1953 0 1 1>}
  NamedRerouteDeclaration(None=@n26) {name=HueShift, nodeColor=<rgba 0 1 0.6484 1>}
  NamedRerouteDeclaration(None=@n18) {name=CloseRangeBlend, nodeColor=<rgba 0.5703 0 1 1>}
  NamedRerouteDeclaration(None=@n15) {name=Close Range Caustics, nodeColor=<rgba 0.1953 0 1 1>}
  n69 = ConstantBiasScale(None=@n66) {bias=-0.5, scale=2}
  n70 = NamedRerouteUsage()
  n71 = Constant() {r=0.5}
  n72 = LinearInterpolate(
   A=ScalarParameter() {defaultValue=995.9999, group=Caustics, parameterName=Caustic Warp Strength, sliderMax=3000}
   B=ScalarParameter() {defaultValue=2400, group=Caustics, parameterName=Caustic Warp Strength 3, sliderMax=3000}
   Alpha=ShadingPathSwitch(Default=QualitySwitch(Default=@n70, Low=@n71), Mobile=@n71)
  )
  n73 = TextureObjectParameter() {parameterName=NoiseTexture, samplerType=SAMPLERTYPE_Alpha, texture=&StormShield2_Noise}
  n74 = ScalarParameter() {defaultValue=43442.4727, parameterName=CausticScaleD}
  n75 = StaticBool() {value=true}
  n76 = ScalarParameter() {defaultValue=-2.8632, parameterName=Bump D}
  n77 = VectorParameter() {parameterName=Panning D, defaultValue=<rgba 0.0077 0.004 -0.0704 1>}
  n78 = ConstantBiasScale(None=ComponentMask(None=MaterialFunctionCall(TextureObject=@n73, TextureSize=@n74, Export Float 4=@n75, BumpOffsetHeight=@n76, In=@n77.RGB) {materialFunction=&WorldAlignedTexture_ScalePanBumpOffset}.XYZ Texture) {a=true}) {bias=-0.5, scale=2}
  n79 = ScalarParameter() {defaultValue=672, group=Caustics, parameterName=Caustic Warp Strength 2, sliderMax=3000}
  NamedRerouteDeclaration(None=Add(A=Multiply(A=@n69, B=@n72), B=Multiply(A=@n78, B=@n79))) {name=Warp, nodeColor=<rgba 1 0 0.3281 1>}
  NamedRerouteDeclaration(None=MaterialFunctionCall(
    TextureObject=TextureObjectParameter() {parameterName=caustictexture, samplerType=SAMPLERTYPE_LinearColor, texture=&StormShield2_M}
    TextureSize=ScalarParameter() {defaultValue=66714.1094, parameterName=CausticScaleE}
    Export Float 4=StaticBool()
    In=VectorParameter() {parameterName=Panning E, defaultValue=<rgba -0.2 0 0 1>}.RGB
   ) {materialFunction=&WorldAlignedTexture_ScalePanBumpOffset}.XYZ Texture) {name=Panning Side noise, nodeColor=<rgba 0 0.9531 1 1>}
  NamedRerouteUsage()
  NamedRerouteUsage()
  NamedRerouteUsage()
  n80 = MaterialFunctionCall(Vector 2=ComponentMask(None=WorldPosition() {worldPositionShaderOffset=WPT_CameraRelative}.XYZ) {g=true, r=true}) {materialFunction=/Engine/Functions/Engine_MaterialFunctions02/Utility/VectorLength}
  n81 = ScalarParameter() {defaultValue=56962.1367, parameterName=FarFade}
  n82 = ScalarParameter() {defaultValue=1.1, parameterName=Buffer}
  n83 = Constant()
  Reroute(None=If(
    A=@n80.V2 Length
    B=Multiply(A=@n81, B=@n82) {constB=1.36, desc=add some buffer}
    "A > B"=Constant4Vector() {constant=<rgba 0 0 -2184973746438144 1>}.RGB
    "A == B"=@n83
    "A < B"=@n83
   ))
  n84 = NamedRerouteUsage()
  n85 = NamedRerouteUsage()
  n86 = Multiply(
   A=ConstantBiasScale(None=NamedRerouteUsage()) {bias=-0.5, scale=1}
   B=LinearInterpolate(B=ScalarParameter() {defaultValue=-0.06, parameterName=HueshiftIntensity}, Alpha=@n85)
  )
  n87 = VectorParameter() {parameterName=Base Background Brightness_Backside, defaultValue=<rgba 0.3297 0 2 0.5>}
  n88 = LinearInterpolate(A=@n87.RGB, B=NamedRerouteUsage(), Alpha=TwoSidedSign())
  n89 = TwoSidedSign()
  n90 = LinearInterpolate(A=@n85, B=NamedRerouteUsage(), Alpha=NamedRerouteUsage())
  n91 = Multiply(
   A=ShadingPathSwitch(Default=QualitySwitch(Default=@n90, Low=@n85, Medium=@n85), Mobile=@n85)
   B=VectorParameter() {parameterName=CausticColor, defaultValue=<rgba 0.0099 0.2374 0.5 1>}.RGB
  )
  n92 = Add(
   A=ShadingPathSwitch(
    Default=MaterialFunctionCall(Hue Shift Percentage=@n86, Texture=@n88) {materialFunction=&HueShift}.Result
    Mobile=LinearInterpolate(A=@n87.RGB, B=@n61.RGB, Alpha=@n89)
   )
   B=@n91
  )
  n93 = Add(
   A=@n92
   B=Multiply(A=VectorParameter() {parameterName=Caustic Spark Color, defaultValue=<rgba 0.4492 10 2.2487 1>}.RGB, B=NamedRerouteUsage())
  )
  n94 = Add(
   A=ShadingPathSwitch(Default=Add(A=@n84, B=@n93), Mobile=@n92)
   B=Multiply(A=Constant3Vector() {constant=<rgba 0.5 3 40 1>}, B=NamedRerouteUsage())
  )
  n95 = ScalarParameter() {defaultValue=1, parameterName=EyeAdaptionWeight}
  n96 = Constant() {r=1.4}
  n97 = Constant() {r=1.7}
  n98 = Constant() {r=1}
  Reroute(None=Multiply(
    A=EyeAdaptationInverse(LightValueInput=@n94, AlphaInput=@n95).EyeAdaptationInverse
    B=MaterialFunctionCall(Value_Morning=@n96, Value_Day=@n97, Value_Evening=@n96, Value_Night=@n98) {materialFunction=&Scalar_Time_of_Day}.Result
   ))
  n99 = Divide(A=Add(A=PixelDepth().R, B=Constant() {r=-100}), B=ScalarParameter() {defaultValue=150, parameterName=NearFade})
  n100 = WorldPosition()
  n101 = ScalarParameter() {defaultValue=25363.0039, parameterName=HeightGradientScale}
  n102 = Saturate(None=ComponentMask(None=Divide(A=@n100.XYZ, B=@n101)) {b=true})
  n103 = Multiply(
   A=Power(Base=OneMinus(None=@n102), Exp=ScalarParameter() {defaultValue=1.7, parameterName=PowerHeight})
   B=Saturate(None=ScalarParameter() {defaultValue=0.4, parameterName=Base Opacity})
  )
  n104 = Add(A=@n80.V2 Length, B=Constant() {r=-100})
  n105 = OneMinus(None=Saturate(None=Divide(A=@n104, B=@n81)))
  n106 = Multiply(
   A=Multiply(A=Saturate(None=@n99), B=Multiply(A=@n103, B=@n105))
   B=ScalarParameter() {defaultValue=1, parameterName=OverallOpacityMult_HighEnd}
  )
  n107 = NamedRerouteUsage()
  n108 = ScalarParameter() {defaultValue=300, parameterName=Depthfade}
  n109 = NamedRerouteUsage()
  Reroute(None=DepthFade(Opacity=Add(A=@n106, B=@n107), FadeDistance=LinearInterpolate(B=@n108, Alpha=@n109)))
  Round(None=@n5)
  n110 = Multiply(A=VectorParameter() {parameterName=Stars, defaultValue=<rgba 0.0122 0.1953 0.0415 1>}.RGB, B=NamedRerouteUsage())
  n111 = Saturate(None=NamedRerouteUsage())
  ShadingPathSwitch(Default=QualitySwitch(Default=LinearInterpolate(A=@n110, B=@n91, Alpha=@n111), Low=@n91), Mobile=@n91)
  SmoothStep(Value=@n5)
  TextureSample()
  VertexInterpolator(VS=ScalarParameter() {bUseCustomPrimitiveData=true, parameterName=MyIndex})
 ]
}
