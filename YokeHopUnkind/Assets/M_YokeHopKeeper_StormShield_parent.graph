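# Parent material for the YokeHopKeeper storm shield: an unlit, translucent, two-sided animated energy wall with parameters for colors, noise and fades.
# World-aligned panning StormShield noises, warped and blended by camera distance and curve-atlas animation, form lightning masks, alongside parallax galaxy noise.
# Color layers separate front/back tints, distance blend, intersection glow and lightning.
# Opacity uses height, near, far and depth fades with a mobile fallback, and an offset drops geometry beyond far range.
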Material /YokeHopUnkind/Assets/M_YokeHopKeeper_StormShield_parent
graph {
 properties {
  bCastRayTracedShadows=false
  bEnableResponsiveAA=true
  bUsedWithNanite=true
  bUsedWithSplineMeshes=true
  blendMode=BLEND_Translucent
  pixelDepthOffsetMode=PDOM_Legacy
  shadingModel=MSM_Unlit
  twoSided=true
  usedShadingModels=Unlit
 }
 nodes [
  n1 = ComponentMask(None=NamedRerouteUsage()) {r=true}
  n2 = Multiply(
   A=Multiply(A=NamedRerouteUsage(), B=LinearInterpolate(Alpha=@n1) {constA=0.6})
   B=VectorParameter() {parameterName=ElectricNoise, defaultValue=<rgba 4.2992 0.6102 9.6038 1>}.RGB
  )
  n3 = LinearInterpolate(
   A=LinearInterpolate(
    A=VectorParameter() {parameterName=Base Background Brightness_Backside, defaultValue=<rgba 0.4283 0.2154 1.0938 0.5>}.RGB
    B=VectorParameter() {parameterName=Base Background Brightness, defaultValue=<rgba 0.2402 0.4529 1.3646 0.5>}.RGB
    Alpha=TwoSidedSign()
   )
   B=LinearInterpolate(
    A=VectorParameter() {parameterName=Distance Blend Color, defaultValue=<rgba 0.0153 0.8954 5 1>}.RGB
    B=AtmosphericLightColor()
    Alpha=ScalarParameter() {defaultValue=0.02, parameterName=Sky Fresnel Blend}
   )
   Alpha=MaterialFunctionCall!(Distance Bias, Distance Divisor=ScalarParameter() {defaultValue=17693.9453, parameterName=DistanceDivide}, Scene Depth, Pixel Depth) {materialFunction=/Game/Effects/Fort_Effects/Materials/_Functions/MF_DistanceBehindTranslucency_01}.Clamped
  )
  Add(A=@n3, B=@n2)
  Add(B=@n3)
  ComponentMask(None=NamedRerouteUsage()) {b=true}
  ComponentMask() {g=true, r=true}
  Constant() {r=0.5}
  Constant() {r=1}
  n4 = WorldPosition()
  n5 = MaterialFunctionCall!(
   TextureObject=TextureObject() {texture=&/GooNoise_M}
   TextureSize=ScalarParameter() {defaultValue=28629.5176, parameterName=CausticScale_near2}
   WorldPosition
   Export Float 4=StaticBool()
   BumpOffsetHeight=ScalarParameter() {defaultValue=-4.0866, parameterName=BumpOffsetGalaxy}
   In=VectorParameter() {parameterName=LightningPanner_Electric1, defaultValue=<rgba 0 0.0064 0 0>}.RGB
   World Space Normal
   ProjectionTransitionContrast=Constant() {r=1}
  ) {materialFunction=/Game/Athena/SafeZone/WorldAlignedTexture_ScalePanBumpOffset}
  n6 = Multiply(A=ComponentMask(None=@n5.XY Texture) {b=true}, B=ScalarParameter() {defaultValue=7708.7295, parameterName=ffset})
  n7 = Multiply(
   A=ScalarParameter() {defaultValue=1.8555, parameterName=MinimumGalaxy}
   B=MaterialFunctionCall(
    TextureObject=TextureObject() {texture=&/GradientGalaxy_M}
    TextureSize=ScalarParameter() {defaultValue=36999.8398, parameterName=CausticScale_Galaxy}
    WorldPosition=Add(A=@n4.XYZ, B=@n6)
    Export Float 4=StaticBool()
    BumpOffsetHeight=ScalarParameter() {defaultValue=-2.5501, parameterName=BumpOffsetGalaxy2}
    In=VectorParameter() {parameterName=LightningPanner_Electric2, defaultValue=<rgba 0 0 0 1>}.RGB
    ProjectionTransitionContrast=Constant() {r=1}
   ) {materialFunction=&WorldAlignedTexture_ScalePanBumpOffset}.XY Texture
  ) {constA=0.0832}
  Divide(A=@n7)
  n8 = LinearInterpolate(
   A=ScalarParameter() {defaultValue=0.65, parameterName=baseOpacity_Back}
   B=ScalarParameter() {defaultValue=0.8566, parameterName=BaseOpacity}
   Alpha=TwoSidedSign()
  )
  n9 = LinearInterpolate(A=@n8, B=Multiply(A=@n8, B=ScalarParameter() {defaultValue=2, parameterName=FarDistanceBaseOpacity}), Alpha=NamedRerouteUsage())
  n10 = NamedRerouteUsage()
  LinearInterpolate(
   A=@n9
   B=Multiply(A=@n9, B=ScalarParameter() {defaultValue=1.5, parameterName=Lightning Opacity})
   Alpha=Saturate(None=ComponentMask(None=@n10) {g=true})
  )
  MaterialFunctionCall!(Center Point World Position) {materialFunction=/Engine/Functions/Engine_MaterialFunctions02/UVs/WorldSpaceAlignedScreenCoordinates}
  n11 = Multiply(
   A=Constant4Vector() {constant=<rgba 8.2617 40 37.1693 1>}.RGB
   B=SphereMask(
    A=ScalarParameter() {defaultValue=6091.0718, parameterName=WaveTop}
    B=WorldPosition().Z
    Radius=ScalarParameter() {defaultValue=3432.531, parameterName=GlowRadius}
    Hardness=ScalarParameter() {defaultValue=-4.0011, parameterName=Glowhardness}
   )
  )
  n12 = NamedRerouteUsage()
  n13 = NamedRerouteUsage()
  Multiply(A=@n11, B=Multiply(A=ComponentMask(None=@n12) {g=true}, B=ComponentMask(None=@n13) {b=true}))
  n14 = NamedRerouteUsage()
  n15 = ComponentMask(None=@n14) {g=true}
  n16 = ComponentMask(None=@n14) {r=true}
  n17 = Power(Base=ComponentMask(None=NamedRerouteUsage()) {b=true}, Exp=ScalarParameter() {defaultValue=2.7986, parameterName=LightningMaskTransition})
  n18 = ScalarParameter() {defaultValue=1.9636, parameterName=LightningMaskintensity}
  n19 = Saturate(None=Power(Base=LinearInterpolate(A=@n15, B=@n16, Alpha=@n17), Exp=@n18))
  Multiply(A=@n19, B=@n18)
  Multiply(A=@n3, B=@n2)
  n20 = NamedRerouteUsage()
  n21 = Multiply(
   A=ScalarParameter() {defaultValue=0.5832, parameterName=Pulse Radius}
   B=LinearInterpolate(Alpha=ComponentMask(None=@n20) {g=true}) {constA=0.66}
  )
  n22 = Subtract(B=@n21) {constA=0}
  n23 = Add(B=@n21) {constA=1}
  n24 = Time()
  n25 = ScalarParameter() {defaultValue=0.0976, parameterName=Shield Animation Speed}
  n26 = CurveAtlasRowParameter(CurveTime=Frac(None=Multiply(A=@n24, B=@n25))) {atlas=&/YokeHopCurveAtlas, curve=&/YokeHopCurve, parameterName=YokeHopCurves}
  n27 = ComponentMask(None=TextureCoordinate()) {g=true}
  NamedRerouteDeclaration(None=SphereMask(A=LinearInterpolate(A=@n22, B=@n23, Alpha=@n26.G) {constA=-0.1, constB=1.2}, B=OneMinus(None=@n27), Radius=@n21) {attenuationRadius=0.1, hardnessPercent=0}) {name=IdleAnimation, nodeColor=<rgba 1 0.0703 0 1>}
  n28 = MaterialFunctionCall(
   Fade Length=ScalarParameter() {defaultValue=8254.0068, parameterName=Far Fade Dist2}
   Fade Offset=ScalarParameter() {defaultValue=6702.0186, parameterName=DistanceFade2, sortPriority=8000000}
  ) {materialFunction=&CameraDepthFade}
  NamedRerouteDeclaration(None=@n28.Result) {name=DistanceFade01, nodeColor=<rgba 1 0.7734 0 1>}
  n29 = ScalarParameter() {parameterName=Dest}
  NamedRerouteDeclaration(None=Multiply(A=ComponentMask(None=@n5.XY Texture) {r=true}, B=Desaturation(None=@n7, Fraction=@n29))) {name=GalaxyTexture, nodeColor=<rgba 0 0.0625 1 1>}
  NamedRerouteDeclaration(None=@n5.XY Texture) {name=CausticNoise1, nodeColor=<rgba 1 0 0.3281 1>}
  n30 = TextureObject() {samplerType=SAMPLERTYPE_LinearColor, texture=&/StormShield_M}
  NamedRerouteDeclaration(None=MaterialFunctionCall!(
    TextureObject=@n30
    TextureSize=VectorParameter() {parameterName=Counterpanner_Size, defaultValue=<rgba 33627.5508 30821.1523 12447.9365 1>}.RGB
    WorldPosition
    Export Float 4=StaticBool()
    In=VectorParameter() {parameterName=LightningPanner_Inter, defaultValue=<rgba 0 -0.01 0.003 0>}.RGB
    World Space Normal
    ProjectionTransitionContrast=Constant() {r=1}
   ) {materialFunction=/Game/Athena/SafeZone/WorldAlignedTexture_ScalePan}.XY Texture) {name=Counterpanning Noise, nodeColor=<rgba 0.0547 1 0 1>}
  NamedRerouteDeclaration(None=CollectionParameter() {collection=&/MPC_YokeHopUnkind, parameterName=IsWaveInfill}) {name=isWaveInfill, nodeColor=<rgba 1 0 0.0234 1>}
  n31 = WorldPosition()
  n32 = Multiply(A=ScalarParameter() {defaultValue=225.9203, parameterName=MainWarpAmount}, B=ComponentMask(None=NamedRerouteUsage()) {r=true})
  n33 = NamedRerouteDeclaration(None=MaterialFunctionCall(
    TextureObject=@n30
    TextureSize=VectorParameter() {parameterName=CausticScaleDetail_Texturing, defaultValue=<rgba 6670.7378 6111 4333.2236 1>}.RGB
    WorldPosition=Add(A=@n31.XYZ, B=@n32)
    Export Float 4=StaticBool()
    In=VectorParameter() {parameterName=DetailnoisePanner, defaultValue=<rgba 0 0.01 -0.002 0>}.RGB
    ProjectionTransitionContrast=Constant() {r=1}
   ) {materialFunction=&WorldAlignedTexture_ScalePan}.XY Texture) {name=DetailNoise, nodeColor=<rgba 0.9219 1 0 1>}
  n34 = MaterialFunctionCall(
   TextureObject=@n30
   TextureSize=VectorParameter() {parameterName=CausticScale_Texturing, defaultValue=<rgba 59411.9688 56780.3945 13895.2842 1>}.RGB
   Export Float 4=StaticBool()
   In=VectorParameter() {parameterName=LightningPanner_Near, defaultValue=<rgba 0 0.03 0.001 0>}.RGB
   ProjectionTransitionContrast=Constant() {r=1}
  ) {materialFunction=&WorldAlignedTexture_ScalePan}
  n35 = Saturate(None=@n28.Result)
  n36 = ScalarParameter() {defaultValue=0.1, parameterName=CounterpanningMin}
  n37 = NamedRerouteUsage()
  n38 = Frac(None=Multiply(A=Time(), B=ScalarParameter() {defaultValue=0.024, parameterName=BlendSpeed}))
  n39 = LinearInterpolate(
   A=ComponentMask(None=@n37) {g=true}
   B=ComponentMask(None=@n37) {r=true}
   Alpha=Saturate(None=CurveAtlasRowParameter(CurveTime=@n38) {atlas=&YokeHopCurveAtlas, curve=&YokeHopCurve, parameterName=YokeHopCurves}.R)
  )
  NamedRerouteDeclaration(None=Multiply(A=LinearInterpolate(A=@n33, B=@n34.XY Texture, Alpha=@n35), B=LinearInterpolate(A=@n36, Alpha=@n39))) {name=Constant Lightning Mask, nodeColor=<rgba 1 0 0.375 1>}
  n40 = ScalarParameter() {parameterName=bias}
  n41 = ScalarParameter() {defaultValue=300, parameterName=Intersection Edge Divisor}
  NamedRerouteDeclaration(None=Power(
    Base=MaterialFunctionCall(Distance Bias=@n40, Distance Divisor=@n41) {materialFunction=&MF_DistanceBehindTranslucency_01}.Inverted
    Exp=ScalarParameter() {defaultValue=5, parameterName=Intersection Edge Exponent}
   )) {name=Edge Glow, nodeColor=<rgba 0.1953 0 1 1>}
  NamedRerouteDeclaration(None=@n34.XY Texture) {name=PrimaryNoise, nodeColor=<rgba 1 0.4453 0 1>}
  NamedRerouteUsage()
  n42 = MaterialFunctionCall(Vector 2=ComponentMask(None=WorldPosition() {worldPositionShaderOffset=WPT_CameraRelative}.XYZ) {g=true, r=true}) {materialFunction=/Engine/Functions/Engine_MaterialFunctions02/Utility/VectorLength}
  n43 = ScalarParameter() {defaultValue=77406.1641, parameterName=FarFade}
  n44 = ScalarParameter() {defaultValue=1.1, parameterName=Buffer}
  n45 = Constant()
  Reroute(None=If(
    A=@n42.V2 Length
    B=Multiply(A=@n43, B=@n44) {constB=1.36, desc=add some buffer}
    "A > B"=Constant3Vector() {constant=<rgba 0 0 -999999 0>}
    "A == B"=@n45
    "A < B"=@n45
   ))
  n46 = NamedRerouteUsage()
  n47 = Add(
   A=LinearInterpolate(A=@n3, B=@n2, Alpha=ScalarParameter() {defaultValue=0.6291, parameterName=Electric})
   B=Multiply(A=VectorParameter() {parameterName=Intersection Color, defaultValue=<rgba 0 25.657 50 1>}.RGB, B=NamedRerouteUsage())
  )
  n48 = NamedRerouteUsage()
  n49 = LinearInterpolate(
   A=VectorParameter() {parameterName=EdgeLightningColor, defaultValue=<rgba 0 3.2514 8.9062 1>}.RGB
   B=VectorParameter() {parameterName=EdgeLightningColorFar, defaultValue=<rgba 0 3.0973 8.4844 1>}.RGB
   Alpha=Saturate(None=Power(Base=@n48))
  )
  n50 = LinearInterpolate(A=Multiply(A=@n49, B=ScalarParameter() {defaultValue=1.1, parameterName=BacksideEdgeMult}), B=@n49, Alpha=TwoSidedSign())
  n51 = LinearInterpolate(A=@n50, B=Multiply(A=@n50, B=ScalarParameter() {defaultValue=2.0093, parameterName=LightningMaskintensityHigh}), Alpha=NamedRerouteUsage())
  Reroute(None=Add(A=Multiply(A=@n11, B=@n46), B=LinearInterpolate(A=@n47, B=@n51, Alpha=@n19)))
  n52 = Divide(A=Add(A=PixelDepth().R, B=Constant() {r=-100}), B=ScalarParameter() {defaultValue=150, parameterName=NearFade})
  n53 = Divide(A=WorldPosition().XYZ, B=ScalarParameter() {defaultValue=30842.082, parameterName=HeightGradientScale})
  n54 = OneMinus(None=Saturate(None=ComponentMask(None=@n53) {b=true}))
  n55 = ScalarParameter() {defaultValue=1.7, parameterName=PowerHeight}
  n56 = Add(A=NamedRerouteUsage(), B=Saturate(None=@n9))
  n57 = QualitySwitch(Default=Multiply(A=Power(Base=@n54, Exp=@n55), B=@n56), Low=@n56)
  n58 = Add(A=@n42.V2 Length, B=Constant() {r=-100})
  n59 = OneMinus(None=Saturate(None=Divide(A=@n58, B=@n43)))
  n60 = Multiply(
   A=Multiply(A=Saturate(None=@n52), B=Multiply(A=@n57, B=@n59))
   B=ScalarParameter() {defaultValue=1, parameterName=OverallOpacityMult_HighEnd}
  )
  n61 = ScalarParameter() {defaultValue=0.2696, parameterName=OverallOpacityMult}
  n62 = ScalarParameter() {defaultValue=250, parameterName=Depthfade}
  n63 = NamedRerouteUsage()
  Reroute(None=DepthFade(Opacity=ShadingPathSwitch(Default=@n60, Mobile=@n61), FadeDistance=LinearInterpolate(B=@n62, Alpha=@n63)))
  Saturate(None=@n7)
  ScalarParameter() {defaultValue=14132.9014, parameterName=CausticScale_near}
  ScalarParameter() {defaultValue=7740.79, parameterName=DetailNoiseScale}
  VertexInterpolator(VS=ScalarParameter() {bUseCustomPrimitiveData=true, parameterName=MyIndex})
 ]
}
